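# Sound cue for the hologram pop-in of keys-and-locks containers, classed as loot-collect audio with item-interact attenuation.
# A quality switch sends the first two levels to a two-layer mix: a random PopIn wave pitched up, plus a random distorted PopUp wave pitched down.
# The last two levels play one of two PopUp waves instead, excluded from branch culling.
# A cue-wide pitch boost applies, and overridden concurrency blocks new instances at the limit.
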SoundCue /Game/Sounds/Foley_Loot/Containers/Keys_And_Locks/Keys_And_Locks_HologramPopIn_Cue
root=QualityLevel_0
nodes [
 {id=Mixer_0, children [Modulator_0,Modulator_1], properties {inputVolume [1,1]}}
 {id=Modulator_0, children [Random_0], properties {pitchMax=1.5, pitchMin=1.5}}
 {id=Modulator_1, children [Random_1], properties {pitchMax=0.8, pitchMin=0.7}}
 {id=QualityLevel_0, children [Mixer_0,Mixer_0,Random_2,Random_2]}
 {id=Random_0, children [WavePlayer_0,WavePlayer_1,WavePlayer_2], properties {weights [1,1,1]}}
 {id=Random_1, children [WavePlayer_5,WavePlayer_6,WavePlayer_7], properties {weights [1,1,1]}}
 {id=Random_2, children [WavePlayer_12, WavePlayer_13], properties {bShouldExcludeFromBranchCulling=true, weights [1,1]}}
 {id=WavePlayer_0, properties {soundWaveAssetPtr=&/Keys_And_Locks_HologramPopIn_01}}
 {id=WavePlayer_1, properties {soundWaveAssetPtr=&/Keys_And_Locks_HologramPopIn_02}}
 {id=WavePlayer_12, properties {soundWaveAssetPtr=&/Keys_And_Locks_HologramPopUp_01}}
 {id=WavePlayer_13, properties {soundWaveAssetPtr=&/Keys_And_Locks_HologramPopUp_02}}
 {id=WavePlayer_2, properties {soundWaveAssetPtr=&/Keys_And_Locks_HologramPopIn_03}}
 {id=WavePlayer_5, properties {soundWaveAssetPtr=&/Keys_And_Locks_HologramPopUp_Distort_01}}
 {id=WavePlayer_6, properties {soundWaveAssetPtr=&/Keys_And_Locks_HologramPopUp_Distort_02}}
 {id=WavePlayer_7, properties {soundWaveAssetPtr=&/Keys_And_Locks_HologramPopUp_Distort_03}}
]
properties {
 attenuationSettings=/Game/Sounds/Attenuation/Fort_Item_Interact_Att
 bOverrideConcurrency=true
 maxDistance=1850
 pitchMultiplier=2
 soundClassObject=&Fort_Loot_Collect
 volumeMultiplier=0.7
 concurrencyOverrides {maxCount=8, resolutionRule=PreventNew}
 concurrencySet [&Weapon_Drop_StopOldest]
}
