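# Looping sound cue for an idle key sound in the Keys and Locks loot container foley, routed to the Fort_Quests sound class.
# A looping wave player feeds a Modulator with an adjusted pitch minimum, and the cue adds a raised pitch multiplier.
# It uses its own attenuation asset and overrides concurrency with a prevent-new rule, although a Weapon_Drop_StopOldest concurrency set is also assigned.
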SoundCue /Game/Sounds/Foley_Loot/Containers/Keys_And_Locks/Keys_And_Locks_KeyIdleLoop_Cue
root=Modulator_0
nodes [
 {id=Modulator_0, children [WavePlayer_0], properties {pitchMin=1}}
 {id=WavePlayer_0, properties {bLooping=true, soundWaveAssetPtr=&/Keys_And_Locks_KeyIdleLoop}}
]
properties {
 attenuationSettings=&/Keys_And_Locks_KeyIdleLoop_Att
 bOverrideConcurrency=true
 maxDistance=1500
 pitchMultiplier=1.5
 soundClassObject=&Fort_Quests
 volumeMultiplier=1
 concurrencyOverrides {resolutionRule=PreventNew}
 concurrencySet [&Weapon_Drop_StopOldest]
}
