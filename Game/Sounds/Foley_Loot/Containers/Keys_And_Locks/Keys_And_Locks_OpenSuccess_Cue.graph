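# Sound cue for the foley of successfully opening a lock with keys, filed under loot container sounds.
# A single wave player feeds a Modulator, which adds slight random variation with the pitch floor held at 1.
# It plays boosted in volume through item-interact attenuation and the loot-collect sound class.
# Its concurrency override prevents new instances at the limit, superseding the assigned stop-oldest weapon-drop concurrency set.
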SoundCue /Game/Sounds/Foley_Loot/Containers/Keys_And_Locks/Keys_And_Locks_OpenSuccess_Cue
root=Modulator_0
nodes [
 {id=Modulator_0, children [WavePlayer_0], properties {pitchMin=1}}
 {id=WavePlayer_0, properties {soundWaveAssetPtr=&/Keys_And_Locks_OpenSuccess}}
]
properties {
 attenuationSettings=/Game/Sounds/Attenuation/Fort_Item_Interact_Att
 bOverrideConcurrency=true
 maxDistance=1850
 soundClassObject=&Fort_Loot_Collect
 volumeMultiplier=1.5
 concurrencyOverrides {resolutionRule=PreventNew}
 concurrencySet [&Weapon_Drop_StopOldest]
}
