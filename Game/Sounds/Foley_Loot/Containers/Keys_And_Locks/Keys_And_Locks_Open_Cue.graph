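# Sound cue for the opening sound of a keys-and-locks loot container, routed to the loot reward sound class with item-interact attenuation.
# A mixer layers two branches, each randomly picking between CaseOpen 01 and 02 waves through its own modulator.
# One branch plays near normal pitch and the other is pitched noticeably down, thickening the sound.
# It references the Weapon_Drop_StopOldest concurrency set.
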SoundCue /Game/Sounds/Foley_Loot/Containers/Keys_And_Locks/Keys_And_Locks_Open_Cue
root=Mixer_0
nodes [
 {id=Mixer_0, children [Modulator_0,Modulator_1], properties {inputVolume [1,1]}}
 {id=Modulator_0, children [Random_0], properties {pitchMin=1}}
 {id=Modulator_1, children [Random_1], properties {pitchMax=0.8, pitchMin=0.7}}
 {id=Random_0, children [WavePlayer_1,WavePlayer_2], properties {weights [1,1]}}
 {id=Random_1, children [WavePlayer_0,WavePlayer_3], properties {weights [1,1]}}
 {id=WavePlayer_0, properties {soundWaveAssetPtr=&/Keys_And_Locks_CaseOpen_02}}
 {id=WavePlayer_1, properties {soundWaveAssetPtr=&Keys_And_Locks_CaseOpen_02}}
 {id=WavePlayer_2, properties {soundWaveAssetPtr=&/Keys_And_Locks_CaseOpen_01}}
 {id=WavePlayer_3, properties {soundWaveAssetPtr=&Keys_And_Locks_CaseOpen_01}}
]
properties {
 attenuationSettings=/Game/Sounds/Attenuation/Fort_Item_Interact_Att
 maxDistance=1850
 soundClassObject=&Fort_Loot_Reward_PSM
 volumeMultiplier=0.9
 concurrencyOverrides {resolutionRule=PreventNew}
 concurrencySet [&Weapon_Drop_StopOldest]
}
