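# Sound cue for the foley heard when keys and locks loot is picked up.
# A single wave player feeds a modulator that slightly lowers pitch at random for variation.
# The cue uses weapon pickup attenuation and the loot collect sound class at reduced volume.
# It joins a weapon-drop concurrency set named for stopping the oldest sound, alongside a prevent-new override setting.
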SoundCue /Game/Sounds/Foley_Loot/Containers/Keys_And_Locks/Keys_And_Locks_PickUp_Cue
root=Modulator_0
nodes [
 {id=Modulator_0, children [WavePlayer_0], properties {pitchMax=1, pitchMin=0.98}}
 {id=WavePlayer_0, properties {soundWaveAssetPtr=&/Keys_And_Locks_PickUp}}
]
properties {
 attenuationSettings=&Fort_Weapon_Pickup_Att
 maxDistance=3500
 soundClassObject=&Fort_Loot_Collect
 volumeMultiplier=0.8
 concurrencyOverrides {resolutionRule=PreventNew}
 concurrencySet [&Weapon_Drop_StopOldest]
}
